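# Translucent unlit heavy-rain overlay for StW storm gameplay cues, on a camera-attached mesh with depth test disabled.
# Camera-vector UVs drive two rotating, crossfaded zooming smoke layers, blended with radial panning streaks, normal-distorted and time-jittered by temporal AA dither.
# Scene depth and distance masks shape opacity, scaled by BP_Blend; its screen derivatives tint color at top quality.
# Color uses inverse eye adaptation, with mobile and lower-quality fallbacks.
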Material /FNGameplayCues/GameplayCueNotifies/StWStorm/M_HeavyRainPost_Ch7
graph {
 properties {
  bDisableDepthTest=true
  bUseTranslucencyVertexFog=false
  bUsedWithNanite=true
  blendMode=BLEND_Translucent
  blendableLocation=BL_SceneColorAfterDOF
  pixelDepthOffsetMode=PDOM_Legacy
  shadingModel=MSM_Unlit
  usedShadingModels=Unlit
 }
 nodes [
  n1 = VectorParameter() {parameterName=Param_1, defaultValue=<rgba 0.2882 0.0466 0.5 0>}
  n2 = SceneDepth(UVs=ScreenPosition().ViewportUV)
  n3 = Saturate(None=Divide(A=ComponentMask(None=@n2.R) {r=true}, B=ScalarParameter() {defaultValue=50}))
  n4 = ScalarParameter() {defaultValue=0.15, parameterName=Param_4}
  n5 = ScalarParameter() {defaultValue=1, parameterName=Param_5}
  n6 = CameraVectorWS()
  n7 = ComponentMask(None=@n6) {g=true, r=true}
  n8 = VectorParameter() {parameterName=UVS3_D, defaultValue=<rgba 7.0064 6.4868 0.05 0.05>}
  n9 = ComponentMask(None=@n8.RGB) {g=true, r=true}
  n10 = TextureSample(UVs=Panner(Coordinate=Divide(A=@n7, B=@n9), Speed=AppendVector(A=@n8.B, B=@n8.B))) {samplerType=SAMPLERTYPE_Normal, texture=/Game/Packages/Fortress_Sky/Materials/2Normal}
  n11 = ComponentMask(None=Multiply(A=Multiply(A=@n10.RGB), B=ScalarParameter() {defaultValue=0.3001, parameterName=NormalDistort})) {g=true, r=true}
  n12 = ComponentMask(None=CameraVectorWS()) {b=true}
  n13 = ScalarParameter() {defaultValue=0.1143, parameterName=AA}
  n14 = Add(
   A=Time()
   B=Multiply(A=MaterialFunctionCall() {materialFunction=&DitherTemporalAA}.Result, B=@n13) {constB=0.2}
  )
  n15 = MaterialFunctionCall() {materialFunction=&DitherTemporalAA}
  n16 = Add(A=Time(), B=Add(A=@n13, B=Multiply(A=@n15.Result, B=@n13) {constB=0.2}) {constA=0.2})
  n17 = Time()
  n18 = QualitySwitch(Default=ShadingPathSwitch(Default=LinearInterpolate(A=@n14, B=@n16), Mobile=@n17) {desc=Migrated from FeatureLevelSwitch}, Low=@n17)
  n19 = Multiply(A=Multiply(A=@n18, B=Sign(None=@n12))) {constB=-0.33}
  n20 = Frac(None=@n19)
  n21 = LinearInterpolate(Alpha=@n20) {constA=3, constB=0.5}
  n22 = Divide(A=@n7, B=ComponentMask(None=VectorParameter() {parameterName=UVS2, defaultValue=<rgba 1.5 1.5 0 0.05>}.RGB) {g=true, r=true})
  n23 = Rotator(Coordinate=Add(A=@n11, B=Multiply(A=@n21, B=@n22))) {centerX=0, centerY=0, speed=0.125}
  n24 = Abs(None=LinearInterpolate(Alpha=@n20) {constA=-1})
  n25 = Frac(None=Add(A=@n19) {constB=0.5})
  n26 = LinearInterpolate(Alpha=@n25) {constA=3, constB=0.5}
  n27 = Rotator(Coordinate=Add(A=@n11, B=Multiply(A=@n26, B=@n22))) {centerX=0, centerY=0, speed=0.1}
  n28 = Abs(None=LinearInterpolate(Alpha=@n25) {constA=-1})
  n29 = Add(
   A=Multiply(A=TextureSample(UVs=@n23) {texture=&T_Inky_Smoke_Tile}.G, B=OneMinus(None=@n24))
   B=Multiply(A=TextureSample(UVs=@n27) {texture=&T_Inky_Smoke_Tile}.G, B=OneMinus(None=@n28))
  )
  n30 = VectorParameter() {parameterName=UVS1, defaultValue=<rgba 5 0.05 0.05 0.1>}
  n31 = ComponentMask(None=@n6) {b=true}
  n32 = Add(
   A=Multiply(
    A=MaterialFunctionCall(Vector or UVs=@n7) {materialFunction=&VectorToRadialValue}.Vector Converted to Angle
    B=@n30.R
   )
   B=Multiply(A=ScalarParameter() {defaultValue=0.0381, parameterName=Distort}, B=@n31)
  )
  n33 = Multiply(A=Tangent!(None=@n31) {period=4}) {constB=-1}
  n34 = Multiply(A=@n30.G, B=@n33)
  n35 = Add(A=Panner(Coordinate=AppendVector(A=@n32, B=@n34), Time=@n18, Speed=AppendVector(A=@n30.B, B=@n30.B)), B=@n11)
  n36 = SphereMask(A=@n33, B=Constant(), Radius=ScalarParameter() {defaultValue=8.1772, parameterName=HeightFade}) {hardnessPercent=0}
  n37 = Multiply(A=@n36, B=@n36)
  n38 = ScalarParameter() {defaultValue=1, parameterName=Param_2}
  n39 = Multiply(
   A=LinearInterpolate(
    A=LinearInterpolate(A=@n4, B=@n5, Alpha=@n29) {constA=0.25, constB=1.5}
    B=TextureSample(UVs=@n35) {texture=&T_Inky_Smoke_Tile}.G
    Alpha=Multiply(A=@n37, B=@n38)
   ) {constA=1}
   B=MaterialFunctionCall!(Distance Bias=ScalarParameter() {defaultValue=3904.6973, parameterName=Distance Devisor}, Distance Divisor, Scene Depth, Pixel Depth) {materialFunction=/Game/Effects/Fort_Effects/Materials/_Functions/MF_DistanceBehindTranslucency_01}.Clamped
  )
  n40 = Add(
   A=Multiply(A=@n3, B=ScalarParameter() {defaultValue=0.5, parameterName=ForIntensity})
   B=Multiply(A=ScalarParameter() {defaultValue=0.5, parameterName=TextureIntensity}, B=Multiply(A=@n3, B=@n39))
  )
  n41 = ScalarParameter() {defaultValue=15, parameterName=Param_3}
  n42 = Clamp(None=Power(Base=Multiply(A=@n40, B=@n41), Exp=ScalarParameter() {defaultValue=0.75, parameterName=FinalGama})) {maxDefault=0.9}
  n43 = Add(A=DDX(Value=@n42), B=DDY(Value=@n42))
  n44 = ScalarParameter() {defaultValue=0.4667, parameterName=DDXY_Exp}
  n45 = Abs(None=Multiply(A=Sign(None=@n43), B=Power(Base=@n43, Exp=@n44)))
  n46 = ScalarParameter() {defaultValue=0.5, parameterName=Standard Depth}
  n47 = LinearInterpolate(
   A=@n1.RGB
   B=VectorParameter() {parameterName=Param_2, defaultValue=<rgba 0 0 -0.1 0>}.RGB
   Alpha=ShadingPathSwitch(Default=QualitySwitch(Default=@n45, Low=@n46, High=@n46, Medium=@n46), Mobile=@n46) {desc=Migrated from FeatureLevelSwitch}
  )
  n48 = ScalarParameter() {defaultValue=0.5, parameterName=EyeAdaptionAmount}
  n49 = ScalarParameter() {defaultValue=1, parameterName=OpacityBig}
  LinearInterpolate(
   B=StaticSwitchParameter(
    True=ShadingPathSwitch(Default=@n47, Mobile=@n1.RGB) {desc=Migrated from FeatureLevelSwitch}
    False=EyeAdaptationInverse(LightValueInput=@n47, AlphaInput=@n48).EyeAdaptationInverse
   ) {parameterName=UseMobileVisuals}
   Alpha=@n49
  ) {constB=0}
  MaterialFunctionCall!(Object Basis 1, Object Basis 2, Object Basis 3, Camera Space Position Offset) {materialFunction=/Engine/Functions/Engine_MaterialFunctions02/WorldPositionOffset/AttachMeshToTheCamera}
  n50 = Multiply(A=Multiply(A=ComponentMask(None=@n42) {r=true}) {constB=0.8}, B=ScalarParameter() {defaultValue=1, parameterName=BP_Blend})
  n51 = ShadingPathSwitch(Default=@n50, Mobile=ScalarParameter() {defaultValue=0.22, parameterName=Param_1}) {desc=Migrated from FeatureLevelSwitch}
  Reroute(None=Multiply(A=@n49, B=StaticSwitchParameter(True=@n51, False=@n50) {parameterName=UseMobileVisuals}))
 ]
}
